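# Masked, unlit Niagara sprite material for a lantern that glows, flickers and dissolves.
# A normal texture drives faked top-down diffuse, specular and particle-tinted Fresnel rim lighting.
# Mask channels blend body color and add glow, with noise-distorted flicker controlled by dynamic parameters (amount, brightness, sharpness, random).
# Particle alpha drives a noisy, vertical-gradient dissolve on 8x-tiled UVs, cutting normal-texture alpha and adding a DissolveColor burning edge.
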Material /DealershipSharedLibrary/Assets/Sets/Lantern/M_DLR_Lantern
graph {
 properties {bUsedWithNiagaraSprites=true, blendMode=BLEND_Masked, pixelDepthOffsetMode=PDOM_Legacy, shadingModel=MSM_Unlit, usedShadingModels=Unlit}
 nodes [
  n1 = Transform(None=CameraVectorWS()) {transformSourceType=TRANSFORMSOURCE_World, transformType=TRANSFORM_Tangent}
  n2 = Constant() {r=2}
  n3 = TextureCoordinate()
  n4 = TextureSampleParameter2D(UVs=@n3) {parameterName=Normal, samplerType=SAMPLERTYPE_LinearColor, texture=&/T_DLR_Lantern_FB_N}
  n5 = MaterialFunctionCall(X=Power(Base=@n4.R), Y=Power(Base=@n4.G), Z=@n4.B) {materialFunction=&MakeFloat3}
  MaterialExpressionVertexColor_0 = ParticleColor()
  n6 = Multiply(
   A=Multiply(
    A=MaterialFunctionCall(Normal Vector=@n5.Result, Camera Vector=@n1, Power=@n2) {materialFunction=&Fresnel_Function}.Result
    B=Constant() {r=1.5}
   )
   B=@MaterialExpressionVertexColor_0.RGB
  )
  n7 = Desaturation(None=@n5.Result)
  n8 = Transform(None=@n5.Result)
  n9 = Constant3Vector() {constant=<rgba 0 0 1 0>}
  n10 = ConstantBiasScale(None=DotProduct(A=@n8, B=@n9))
  MaterialExpressionConstantClamp_1 = Saturate(None=Multiply(A=Constant() {r=2}, B=Multiply(A=@n7, B=@n10)))
  n11 = Transform(None=@n9)
  n12 = Transform(None=CameraVectorWS()) {transformSourceType=TRANSFORMSOURCE_World, transformType=TRANSFORM_Tangent}
  n13 = DotProduct(A=@n8, B=Normalize(VectorInput=Add(A=@n11, B=@n12)))
  n14 = Constant() {r=16}
  n15 = Add(A=@MaterialExpressionConstantClamp_1, B=Multiply(A=Constant() {r=0.5}, B=Power(Base=@n13, Exp=@n14)))
  n16 = Multiply(A=@MaterialExpressionVertexColor_0.RGB, B=@MaterialExpressionVertexColor_0.RGB)
  n17 = TextureSampleParameter2D(UVs=@n3) {parameterName=Mask, texture=&/T_DLR_Lantern_FB_M}
  MaterialExpressionConstantClamp_2 = Saturate(None=Multiply(A=@n17.G, B=Constant() {r=2}))
  n18 = Constant3Vector() {constant=<rgba 2 0.2 0.1 0>}
  n19 = Constant() {r=0.4}
  n20 = LinearInterpolate(
   A=Multiply(A=Desaturation(None=@n18, Fraction=@n19), B=Constant() {r=1.5})
   B=Desaturation(None=@n16, Fraction=Constant() {r=0.3})
   Alpha=@MaterialExpressionConstantClamp_2
  )
  MaterialExpressionConstantClamp_0 = Saturate(None=Power(Base=Multiply(A=@n15, B=@n20), Exp=Constant() {r=1.55}))
  n21 = DynamicParameter() {paramNames [FlickerAmount, Brightness, Sharpness, Random]}
  MaterialExpressionConstantClamp_3 = Saturate(None=Multiply(A=Add(A=@n17.G, B=@n17.G), B=@n21.Sharpness))
  n22 = Time()
  n23 = Add(
   A=Panner(Coordinate=TextureCoordinate() {uTiling=0, vTiling=0}, Time=Multiply(A=@n22, B=@n21.FlickerAmount)) {speedX=1, speedY=1}
   B=@n21.Random
  )
  n24 = Multiply(
   A=ComponentMask(None=TextureSample(UVs=@n23) {samplerType=SAMPLERTYPE_Normal, texture=/DealershipSharedLibrary/FX/Textures/Noise/T_VCVFX_Water02_N}.RGB) {g=true, r=true}
   B=Constant() {r=8}
  )
  n25 = Add(A=Multiply(A=Frac(None=@n24), B=Constant() {r=0.1}), B=@n3)
  n26 = Multiply(
   A=@n21.Brightness
   B=Multiply(
    A=@n17.G
    B=TextureSampleParameter2D(UVs=@n25) {parameterName=Mask, texture=&T_DLR_Lantern_FB_M}.B
   )
  )
  n27 = Add(A=@MaterialExpressionConstantClamp_0, B=Multiply(A=@n16, B=Add(A=@MaterialExpressionConstantClamp_3, B=@n26)))
  n28 = Multiply(A=@MaterialExpressionVertexColor_0.A, B=Constant() {r=1.4})
  n29 = Frac(None=Multiply(A=TextureCoordinate(), B=Constant() {r=8}))
  n30 = Panner(Coordinate=@n29) {speedY=0.0075}
  n31 = Panner(Coordinate=@n29) {speedY=0.01}
  n32 = Desaturation(None=Add(
    A=TextureSample(UVs=@n30) {texture=/DealershipSharedLibrary/FX/Textures/Noise/T_VCVFX_TrailMask_RGB}.RGB
    B=TextureSample(UVs=@n31) {texture=/DealershipSharedLibrary/FX/Textures/Noise/T_VCVFX_NoiseCloud_RGB}.RGB
   ))
  n33 = ComponentMask(None=@n29) {g=true}
  n34 = Multiply(A=Multiply(A=Multiply(A=@n32, B=@n33), B=@n28), B=Constant() {r=8})
  MaterialExpressionConstantClamp_5 = Saturate(None=ConstantBiasScale(None=@n33) {bias=-0.1, scale=3})
  n35 = Multiply(
   A=OneMinus(None=@n28)
   B=Multiply(
    A=Multiply(A=@MaterialExpressionConstantClamp_5, B=@MaterialExpressionConstantClamp_5)
    B=TextureSample(UVs=@n29) {texture=&T_VCVFX_NoiseCloud_RGB}.R
   )
  )
  n36 = Power(Base=Add(A=@n28, B=Add(A=@n34, B=@n35)), Exp=Constant() {r=15})
  MaterialExpressionConstantClamp_4 = Saturate(None=ConstantBiasScale(None=@n36) {bias=-0.1, scale=1})
  n37 = OneMinus(None=@MaterialExpressionConstantClamp_4)
  n38 = VectorParameter() {parameterName=DissolveColor, defaultValue=<rgba 1 0.1728 0 20>}
  n39 = Multiply(A=@n38.RGB, B=@n38.RGB)
  MaterialFunctionCall(Emissive=Add(A=Add(A=@n6, B=@n27), B=Multiply(A=@n37, B=@n39))) {materialFunction=&MF_FX_EmissiveAdjustments}
  MaterialExpressionConstantClamp_6 = Saturate(None=Add(A=@n36, B=@MaterialExpressionConstantClamp_4))
  Multiply(A=@n4.A, B=@MaterialExpressionConstantClamp_6)
 ]
}
