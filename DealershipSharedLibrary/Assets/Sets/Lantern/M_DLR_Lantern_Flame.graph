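# Translucent, unlit, two-sided material for the lantern flame in the Dealership library, enabled for Niagara mesh particles.
# The FlameColor parameter is multiplied by itself and feeds the emissive input of MF_FX_EmissiveAdjustments.
# A vertically panning flow noise distorts UVs around an offset center, and an inverted squared-distance gradient with a high power carves a procedural flame shape.
# That mask is multiplied by particle alpha for transparency.
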Material /DealershipSharedLibrary/Assets/Sets/Lantern/M_DLR_Lantern_Flame
graph {
 properties {
  bUsedWithNiagaraMeshParticles=true
  blendMode=BLEND_Translucent
  previewMesh=&/SM_Lantern
  shadingModel=MSM_Unlit
  twoSided=true
  usedShadingModels=Unlit
 }
 nodes [
  n1 = VectorParameter() {parameterName=FlameColor, defaultValue=<rgba 1 0.4 0 32>}
  MaterialFunctionCall(Emissive=Multiply(A=@n1.RGB, B=@n1.RGB)) {materialFunction=&MF_FX_EmissiveAdjustments}
  n2 = TextureCoordinate() {uTiling=0.3, vTiling=0.5}
  n3 = ComponentMask(None=TextureSample(UVs=Panner(Coordinate=@n2) {speedY=1.25}) {texture=/DealershipSharedLibrary/FX/Textures/Noises/T_DLR_Flow03_D}.RGB) {g=true, r=true}
  n4 = Constant() {r=2}
  n5 = Add(A=Add(A=TextureCoordinate() {uTiling=2, vTiling=2}, B=Multiply(A=@n3, B=@n4)), B=Constant2Vector() {g=-2, r=-1.5})
  MaterialExpressionConstantClamp_1 = Saturate(None=DotProduct(A=@n5, B=@n5))
  MaterialExpressionConstantClamp_0 = Saturate(None=Power(Base=OneMinus(None=@MaterialExpressionConstantClamp_1), Exp=Constant() {r=16}))
  Multiply(A=@MaterialExpressionConstantClamp_0, B=ParticleColor().A)
 ]
}
