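# Unlit, masked lantern material for Niagara mesh particles, with fake lighting, an animated inner glow and a dissolve.
# Particle color tints a fixed-direction fake light and Fresnel rim, while a panning water normal warps a radial LightColor glow weighted by a vertical gradient.
# Vertex alpha separates this shading from near-black areas.
# Particle alpha thresholds a randomly offset tile texture, producing the Opacity value and bright dissolve edges before the emissive adjustment function.
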Material /DealershipSharedLibrary/Assets/Sets/Lantern/M_Lantern
graph {
 properties {
  bUsedWithNiagaraMeshParticles=true
  blendMode=BLEND_Masked
  previewMesh=&/SM_Lantern
  shadingModel=MSM_Unlit
  usedShadingModels=Unlit
 }
 nodes [
  Constant() {r=0.564}
  Constant() {r=0.912}
  n1 = Constant3Vector() {constant=<rgba 0 0 1 0>}
  n2 = Transform(None=Normalize(VectorInput=Constant3Vector() {constant=<rgba 0.2 0.2 1 0>})) {transformSourceType=TRANSFORMSOURCE_World, transformType=TRANSFORM_Tangent}
  MaterialExpressionConstantClamp_2 = Saturate(None=ConstantBiasScale(None=DotProduct(A=@n1, B=@n2)))
  n3 = ConstantBiasScale(None=Power(Base=@MaterialExpressionConstantClamp_2, Exp=Constant() {r=4})) {bias=0.1, scale=2}
  MaterialExpressionMeshEmitterVertexColor_1 = ParticleColor()
  n4 = Multiply(A=@MaterialExpressionMeshEmitterVertexColor_1.RGB) {constB=2}
  n5 = Fresnel()
  n6 = VectorParameter() {parameterName=LightColor, defaultValue=<rgba 1 0.5 0.1 1>}
  n7 = ScalarParameter() {defaultValue=2, parameterName=LanternLightBrightness}
  n8 = Panner(Coordinate=TextureCoordinate() {uTiling=0, vTiling=0}) {speedX=1, speedY=1}
  n9 = ScalarParameter() {parameterName=Random}
  n10 = ComponentMask(None=TextureSample(UVs=Add(A=@n8, B=@n9)) {samplerType=SAMPLERTYPE_Normal, texture=/DealershipSharedLibrary/FX/Textures/Normal/T_DLR_AquaticWater_N}.RGB) {g=true, r=true}
  n11 = Add(A=Add(A=TextureCoordinate() {coordinateIndex=2}, B=Multiply(A=@n10) {constB=0.1})) {constB=-0.5}
  n12 = Power(Base=OneMinus(None=DotProduct(A=@n11, B=@n11)), Exp=Constant() {r=8})
  n13 = Power(Base=ComponentMask(None=TextureCoordinate()) {g=true}, Exp=Constant() {r=3})
  MaterialExpressionConstantClamp_3 = Saturate(None=Multiply(A=Add(A=@n12, B=@n13), B=@n13))
  n14 = Add(
   A=Add(A=Multiply(A=@n3, B=@MaterialExpressionMeshEmitterVertexColor_1.RGB), B=Multiply(A=@n4, B=@n5))
   B=Multiply(A=Multiply(A=@n6.RGB, B=@n7), B=@MaterialExpressionConstantClamp_3)
  )
  n15 = Ceil(None=VertexColor().A)
  n16 = VectorParameter() {parameterName=LightColor, defaultValue=<rgba 1 0.5 0.1 1>}
  n17 = ScalarParameter() {defaultValue=1024, parameterName=Emissive Dissolve}
  n18 = TextureSample(UVs=Add(A=TextureCoordinate() {coordinateIndex=2}, B=ScalarParameter() {parameterName=Random})) {texture=/Game/Effects/Fort_Effects/Textures/T_Smooth_tile}
  n19 = LinearInterpolate(Alpha=ParticleColor().A) {constB=1.5}
  n20 = Ceil(None=Subtract(A=Add(A=@n18.RGB) {constB=0.45}, B=@n19))
  MaterialFunctionCall(Emissive=LinearInterpolate(A=LinearInterpolate(B=@n14, Alpha=@n15) {constA=0.05}, B=Multiply(A=@n16.RGB, B=@n17) {constB=4096}, Alpha=Saturate(None=@n20))) {materialFunction=&MF_FX_EmissiveAdjustments}
  MaterialExpressionConstantClamp_0 = Saturate(None=Subtract(A=@n19, B=@n18.RGB))
  NamedRerouteDeclaration(None=@MaterialExpressionConstantClamp_0) {name=Opacity, nodeColor=<rgba 1 0.375 0 1>}
  NamedRerouteUsage()
 ]
}
